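# Unlit prototype Landmass brush material that, as its name indicates, caches a landscape brush distance field.
# It calls LSBrush_DistanceField at a center and eight offset UVs, using UseBlur to blend toward their nine-tap average.
# The distance is scaled by MPC_Erosion resolution and grid size, then packed with PackRG8Height alongside a channel from the unscaled value.
# Separate fragments hold curl noise, mesh depth and jump-flood distance math, none feeding that packing.
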Material /Landmass/Landscape/OldPrototype_BP/Mats/Brushes/CacheDistanceField
graph {
 properties {
  bAllowNegativeEmissiveColor=true
  blendMode=BLEND_AlphaComposite
  pixelDepthOffsetMode=PDOM_Legacy
  shadingModel=MSM_Unlit
  usedShadingModels=Unlit
 }
 nodes [
  n1 = Constant()
  n2 = MaterialFunctionCall!(UV Offset=@n1) {materialFunction=/Landmass/Landscape/OldPrototype_BP/Mats/MF/LSBrush_DistanceField}
  n3 = AppendVector(A=@n2.UV Distance, B=@n2.UV Distance)
  n4 = ScalarParameter() {defaultValue=1, parameterName=Bluroffset}
  n5 = Divide(
   A=@n4
   B=AppendVector(
    A=CollectionParameter() {collection=/Landmass/Landscape/BlueprintBrushes/MPC/MPC_Erosion, parameterName=ResX}
    B=CollectionParameter() {collection=&MPC_Erosion, parameterName=ResY}
   )
  )
  n6 = Multiply(A=@n5, B=Constant2Vector() {r=1})
  n7 = MaterialFunctionCall(UV Offset=Add(A=@n1, B=@n6)) {materialFunction=&LSBrush_DistanceField}
  n8 = MaterialFunctionCall(UV Offset=Add(A=@n1, B=Multiply(A=@n6) {constB=-1})) {materialFunction=&LSBrush_DistanceField}
  n9 = Add(A=AppendVector(A=@n7.UV Distance, B=@n7.UV Distance), B=AppendVector(A=@n8.UV Distance, B=@n8.UV Distance))
  n10 = Multiply(A=@n5, B=Constant2Vector() {g=1})
  n11 = MaterialFunctionCall(UV Offset=Add(A=@n1, B=@n10)) {materialFunction=&LSBrush_DistanceField}
  n12 = MaterialFunctionCall(UV Offset=Add(A=@n1, B=Multiply(A=@n10) {constB=-1})) {materialFunction=&LSBrush_DistanceField}
  n13 = Add(A=AppendVector(A=@n11.UV Distance, B=@n11.UV Distance), B=AppendVector(A=@n12.UV Distance, B=@n12.UV Distance))
  n14 = Divide(
   A=@n4
   B=AppendVector(
    A=CollectionParameter() {collection=&MPC_Erosion, parameterName=ResX}
    B=CollectionParameter() {collection=&MPC_Erosion, parameterName=ResY}
   )
  )
  n15 = Multiply(A=@n14, B=Constant2Vector() {g=1, r=1})
  n16 = MaterialFunctionCall(UV Offset=Add(A=@n1, B=@n15)) {materialFunction=&LSBrush_DistanceField}
  n17 = Constant2Vector() {g=1, r=-1}
  n18 = MaterialFunctionCall(UV Offset=Add(A=@n1, B=Multiply(A=@n15, B=@n17) {constB=-1})) {materialFunction=&LSBrush_DistanceField}
  n19 = Add(A=AppendVector(A=@n16.UV Distance, B=@n16.UV Distance), B=AppendVector(A=@n18.UV Distance, B=@n18.UV Distance))
  n20 = Multiply(A=@n14, B=Constant2Vector() {g=-1, r=-1})
  n21 = MaterialFunctionCall(UV Offset=Add(A=@n1, B=@n20)) {materialFunction=&LSBrush_DistanceField}
  n22 = Constant2Vector() {g=1, r=-1}
  n23 = MaterialFunctionCall(UV Offset=Add(A=@n1, B=Multiply(A=@n20, B=@n22) {constB=-1})) {materialFunction=&LSBrush_DistanceField}
  n24 = Add(A=AppendVector(A=@n21.UV Distance, B=@n21.UV Distance), B=AppendVector(A=@n23.UV Distance, B=@n23.UV Distance))
  n25 = Add(A=Add(A=Add(A=@n9, B=@n13), B=Add(A=@n19, B=@n24)), B=@n3)
  n26 = MaterialFunctionCall(Float2=LinearInterpolate(A=@n3, B=Divide(A=@n25) {constB=9}, Alpha=ScalarParameter() {parameterName=UseBlur})) {materialFunction=&BreakOutFloat2Components}
  n27 = CollectionParameter() {collection=&MPC_Erosion, parameterName=ResX}
  n28 = CollectionParameter() {collection=&MPC_Erosion, parameterName=ResY}
  n29 = Multiply(
   A=Multiply(A=@n26.R, B=Max(A=@n27, B=@n28))
   B=CollectionParameter() {collection=&MPC_Erosion, parameterName=GridSize}
  ) {constB=10}
  n30 = LinearInterpolate(A=@n26.G, B=@n26.G, Alpha=@n2.Mask)
  n31 = MaterialFunctionCall(Float2=MaterialFunctionCall!(Height Int -32k to 32k=@n30) {materialFunction=/Landmass/Landscape/OldPrototype_BP/Mats/MF/PackRG8Height}.XY) {materialFunction=&BreakOutFloat2Components}
  AppendVector(
   A=MaterialFunctionCall(Height Int -32k to 32k=Multiply(A=@n29) {constB=0.5}) {materialFunction=&PackRG8Height}.XY
   B=@n31.R
  )
  n32 = TextureCoordinate()
  n33 = ScalarParameter() {parameterName=Seed1}
  n34 = VectorParameter() {parameterName=Curl, defaultValue=<rgba 3 0.025 24 -0.0025>}
  n35 = VectorNoise(World Position=Multiply(A=AppendVector(A=@n32, B=@n33), B=@n34.R)) {noiseFunction=VNF_CurlALU}
  n36 = Divide(A=@n34.G) {constB=100}
  n37 = TextureCoordinate()
  n38 = ScalarParameter() {parameterName=Seed2}
  n39 = VectorNoise(World Position=Multiply(A=AppendVector(A=@n37, B=@n38), B=@n34.B)) {noiseFunction=VNF_CurlALU}
  n40 = Divide(A=@n34.A) {constB=100}
  ComponentMask(None=Add(A=Multiply(A=@n35, B=@n36), B=Multiply(A=@n39, B=@n40))) {g=true, r=true}
  Divide(
   A=Subtract(B=TextureSampleParameter2D() {parameterName=MeshDepth, texture=/Engine/EngineResources/Black_Low.Black}.R) {constA=50000}
   B=CollectionParameter() {collection=&MPC_Erosion, parameterName=ZScale}
  )
  LinearInterpolate(A=ScalarParameter() {defaultValue=1394, parameterName=Z}, B=@n30, Alpha=ScalarParameter() {parameterName=UseMeshDepth})
  MaterialFunctionCall!(TextureObject, UV, Vector, Distance, Steps) {materialFunction=/Engine/Functions/Engine_MaterialFunctions01/Texturing/MotionBlur-Texture}
  n41 = MaterialFunctionCall() {materialFunction=/Landmass/Landscape/OldPrototype_BP/Mats/MF/PointSampleUVs}
  n42 = Subtract(A=@n41.TexelCenterUVs, B=ComponentMask() {g=true, r=true})
  n43 = TextureSampleParameter2D!(UVs=@n41.TexelCenterUVs, Level=Constant(), Apply View MipBias) {mipValueMode=TMVM_MipLevel, parameterName=JumpFloodRT, texture=&Black}
  n44 = LinearInterpolate(Alpha=@n43.A) {constA=1, constB=-1}
  n45 = LinearInterpolate(Alpha=ScalarParameter() {parameterName=Invert}) {constA=1, constB=-1}
  Multiply(A=Multiply(
    A=MaterialFunctionCall(Vector 2=@n42) {materialFunction=/Engine/Functions/Engine_MaterialFunctions02/Utility/VectorLength}.V2 Length
    B=Multiply(A=@n44, B=@n45)
   )) {constB=-1}
  Multiply(A=@n43.A, B=ScalarParameter() {parameterName=UseCaptureDepth})
  OneMinus(None=@n31.G)
  TextureCoordinate()
  VectorParameter() {desc="A= Invert Shape?", parameterName=X, defaultValue=<rgba 1 0 0 0>}
 ]
}
