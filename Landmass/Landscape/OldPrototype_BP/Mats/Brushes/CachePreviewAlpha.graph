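# Unlit alpha-composite material in the old Landmass landscape prototype's brush set, apparently for caching a preview alpha from height render targets.
# It samples the current and previous height targets, takes the absolute difference of their red channels, divides by a Falloff parameter and saturates it into a change mask.
# The current height and that mask are appended into a two-channel value, with negative emissive values allowed.
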Material /Landmass/Landscape/OldPrototype_BP/Mats/Brushes/CachePreviewAlpha
graph {
 properties {
  bAllowNegativeEmissiveColor=true
  blendMode=BLEND_AlphaComposite
  pixelDepthOffsetMode=PDOM_Legacy
  shadingModel=MSM_Unlit
  usedShadingModels=Unlit
 }
 nodes [
  n1 = TextureSampleParameter2D() {parameterName=HeightRT, texture=/Engine/EngineResources/Black_Low.Black}
  n2 = Abs(None=Subtract(A=TextureSampleParameter2D() {parameterName=HeightRTPrev, texture=&Black}.R, B=@n1.R))
  n3 = ScalarParameter() {defaultValue=256, parameterName=Falloff}
  AppendVector(A=@n1.R, B=Saturate(None=Divide(A=@n2, B=@n3) {constB=5000}))
 ]
}
