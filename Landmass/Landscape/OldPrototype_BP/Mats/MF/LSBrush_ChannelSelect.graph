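# Material function from the Landmass old landscape brush prototype that switches brush output between height rendering and channel-masked layer alpha.
# The HeightMapRender scalar, defaulting to 1, lerps Emissive from Layer Alpha times ChannelMask RGB to the Height RGB input.
# The same scalar lerps Opacity from Layer Alpha times ChannelMask alpha to a constant 1, so height mode is fully opaque.
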MaterialFunction /Landmass/Landscape/OldPrototype_BP/Mats/MF/LSBrush_ChannelSelect
graph {nodes [
  n1 = VectorParameter() {parameterName=ChannelMask}
  n2 = FunctionInput() {bUsePreviewValueAsDefault=true, inputName=Layer Alpha, inputType=FunctionInput_Scalar, sortPriority=1}
  n3 = MaterialFunctionCall() {materialFunction=&MakeFloat3}
  n4 = ScalarParameter() {defaultValue=1, parameterName=HeightMapRender}
  FunctionOutput(None=LinearInterpolate(A=Multiply(A=@n2, B=@n1.RGB), B=FunctionInput(Preview=@n3.Result) {bUsePreviewValueAsDefault=true, inputName=Height RGB}, Alpha=@n4)) {outputName=Emissive}
  n5 = Constant() {r=1}
  FunctionOutput(None=LinearInterpolate(A=Multiply(A=@n2, B=@n1.A), B=@n5, Alpha=@n4)) {outputName=Opacity, sortPriority=1}
  OneMinus(None=@n5)
 ]}
