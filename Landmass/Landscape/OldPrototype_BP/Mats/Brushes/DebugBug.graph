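# DebugBug is an unlit alpha-composite debug material among the old prototype Landmass landscape brushes, apparently a test of height-channel handling.
# It unpacks a black texture's RG8 height into X and copies green from the HeightRT parameter, point-sampled at texel centers, into Y and Z.
# That vector feeds LSBrush_ChannelSelect as Height RGB, with Layer Alpha wired to an empty reroute.
# Negative emissive values are allowed.
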Material /Landmass/Landscape/OldPrototype_BP/Mats/Brushes/DebugBug
graph {
 properties {
  bAllowNegativeEmissiveColor=true
  blendMode=BLEND_AlphaComposite
  pixelDepthOffsetMode=PDOM_Legacy
  shadingModel=MSM_Unlit
  usedShadingModels=Unlit
 }
 nodes [
  n1 = TextureSample() {texture=&Black}
  n2 = TextureSampleParameter2D(UVs=MaterialFunctionCall() {materialFunction=/Landmass/Landscape/OldPrototype_BP/Mats/MF/PointSampleUVs}.TexelCenterUVs) {parameterName=HeightRT, texture=/Engine/EngineResources/Black_Low.Black}
  MaterialFunctionCall!(
   Height RGB=MaterialFunctionCall(X=MaterialFunctionCall!(RG=@n1.RGB) {materialFunction=/Landmass/Landscape/OldPrototype_BP/Mats/MF/UnpackRG8Height}.Result, Y=@n2.G, Z=@n2.G) {materialFunction=&MakeFloat3}.Result
   Layer Alpha=Reroute()
  ) {materialFunction=/Landmass/Landscape/OldPrototype_BP/Mats/MF/LSBrush_ChannelSelect}
 ]
}
