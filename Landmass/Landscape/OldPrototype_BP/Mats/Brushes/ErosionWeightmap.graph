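# Unlit additive material for the prototype Landmass erosion brush that computes a weightmap from erosion activity.
# Height change between current and previous render targets, split into Deposition and Erosion, combines with velocity-texture terms, with ScatterParticles adjusting the mix.
# ChannelMask routes the result into channels, and negative emissive is allowed.
# The graph also carries jump-flood distance-field shape, falloff and blend-mode branches, with several unconnected fragments.
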Material /Landmass/Landscape/OldPrototype_BP/Mats/Brushes/ErosionWeightmap
graph {
 properties {
  bAllowNegativeEmissiveColor=true
  blendMode=BLEND_Additive
  pixelDepthOffsetMode=PDOM_Legacy
  shadingModel=MSM_Unlit
  usedShadingModels=Unlit
 }
 nodes [
  Constant() {r=1}
  n1 = CollectionParameter() {collection=/Landmass/Landscape/BlueprintBrushes/MPC/MPC_Erosion, parameterName=TotalSizeX}
  n2 = ScalarParameter() {defaultValue=1, desc=Tangent, parameterName=FalloffTangent}
  Divide(A=@n1, B=@n2)
  Divide()
  n3 = TextureSampleParameter2D() {parameterName=HeightRT, texture=/Engine/EngineResources/Black_Low.Black}
  n4 = TextureCoordinate()
  n5 = ScalarParameter() {defaultValue=2, parameterName=T}
  n6 = ComponentMask(None=TextureSampleParameter2D(UVs=Multiply(A=@n4, B=@n5)) {desc="Current Texel (upper left)", parameterName=BrushRT, texture=/Engine/MapTemplates/TilingNoise05}.R) {r=true}
  n7 = Multiply(A=@n6, B=ScalarParameter() {defaultValue=256, parameterName=BrushTexHeight})
  n8 = Add(A=@n7, B=ScalarParameter() {defaultValue=1394, parameterName=Z})
  n9 = TextureSampleParameter2D() {parameterName=JumpFloodRT, texture=&Black}
  n10 = Subtract(A=TextureCoordinate(), B=ComponentMask(None=@n9.RGB) {g=true, r=true})
  n11 = LinearInterpolate(Alpha=@n9.A) {constA=1, constB=-1}
  n12 = VectorParameter() {desc="A= Invert Shape?", parameterName=X, defaultValue=<rgba 1 0 0 0>}
  n13 = Multiply(A=Multiply(
    A=MaterialFunctionCall(Vector 2=@n10) {materialFunction=/Engine/Functions/Engine_MaterialFunctions02/Utility/VectorLength}.V2 Length
    B=Multiply(A=@n11, B=@n12.A)
   )) {constB=-1}
  n14 = Multiply(
   A=CollectionParameter() {collection=&MPC_Erosion, parameterName=ResX}
   B=CollectionParameter() {collection=&MPC_Erosion, parameterName=GridSize}
  ) {desc=TotalSize}
  n15 = Divide(A=ScalarParameter() {parameterName=EdgeCenterOffset}, B=@n14)
  n16 = Divide(A=ScalarParameter() {defaultValue=1, desc=Tangent, parameterName=EdgeWidth}, B=@n14)
  n17 = Divide(A=Subtract(A=@n13, B=Subtract(A=@n15, B=@n16)), B=@n16)
  n18 = Multiply(A=CollectionParameter() {collection=&MPC_Erosion, parameterName=ResX}) {constB=10}
  n19 = MaterialFunctionCall!(A=@n17, B=Constant() {r=1}, K=Divide(A=VectorParameter() {parameterName=CurveParams, defaultValue=<rgba 1 1 1 1>}.R, B=@n18)) {materialFunction=/Landmass/Landscape/OldPrototype_BP/Mats/MF/SmoothIntersections}
  n20 = VectorParameter() {parameterName=CurveParams, defaultValue=<rgba 1 1 1 1>}
  n21 = Saturate(None=MaterialFunctionCall(A=@n19.Smooth Min, B=Constant(), K=Divide(A=@n20.G, B=@n18)) {materialFunction=&SmoothIntersections}.Smooth Max)
  n22 = LinearInterpolate(A=@n3.R, B=@n8, Alpha=@n21)
  n23 = Multiply(A=Max(A=@n8) {constB=0}, B=@n19.Smooth Min)
  n24 = VectorParameter() {parameterName=CurveParams, defaultValue=<rgba 1 1 1 1>}
  n25 = Multiply(A=@n8, B=@n21)
  n26 = Multiply(A=@n13) {constB=-1}
  n27 = Constant()
  n28 = VectorParameter() {parameterName=CurveParams, defaultValue=<rgba 1 1 1 1>}
  n29 = Divide(A=Divide(A=@n28.G, B=@n1))
  n30 = ScalarParameter() {parameterName=UseMeshDepth}
  n31 = Multiply(A=Divide(A=ScalarParameter() {defaultValue=1394, parameterName=Z}, B=@n14), B=@n2)
  n32 = Subtract(A=@n15, B=@n31)
  n33 = Subtract(
   A=LinearInterpolate(
    A=@n26
    B=MaterialFunctionCall(A=@n26, B=@n27, K=@n29) {materialFunction=&SmoothIntersections}.Smooth Max
    Alpha=@n30
   )
   B=@n32
  )
  n34 = ScalarParameter() {defaultValue=1394, parameterName=Z}
  n35 = Add(A=@n7, B=Multiply(A=Divide(A=@n33, B=@n31), B=@n34))
  n36 = VectorParameter() {parameterName=CurveParams, defaultValue=<rgba 1 1 1 1>}
  n37 = TextureSampleParameter2D() {parameterName=HeightRT, texture=&Black}
  n38 = MaterialFunctionCall(A=@n37.R, B=@n35, K=@n36.R) {materialFunction=&SmoothIntersections}
  n39 = Divide(A=Divide(A=@n28.R, B=@n1))
  n40 = Subtract(
   A=LinearInterpolate(
    A=@n13
    B=MaterialFunctionCall(A=@n13, B=@n27, K=@n39) {materialFunction=&SmoothIntersections}.Smooth Min
    Alpha=@n30
   )
   B=@n32
  )
  n41 = Add(A=@n7, B=Multiply(A=Divide(A=@n40, B=@n31), B=@n34))
  n42 = TextureSampleParameter2D() {parameterName=HeightRT, texture=&Black}
  n43 = Min(A=@n37.R, B=Add(A=@n35, B=@n42.R))
  n44 = Add(A=@n41, B=@n42.R)
  LinearInterpolate(
   A=MaterialFunctionCall!(
    AlphaBlend=@n22
    Min=Min(A=@n3.R, B=@n22)
    Max=MaterialFunctionCall(A=@n3.R, B=@n23, K=@n24.G) {materialFunction=&SmoothIntersections}.Smooth Max
    Additive=Add(A=@n3.R, B=@n25)
   ) {materialFunction=/Landmass/Landscape/OldPrototype_BP/Mats/MF/BrushBlendModes_Manual}.Result
   B=MaterialFunctionCall(
    AlphaBlend=MaterialFunctionCall(A=@n38.Smooth Min, B=@n41, K=@n36.G) {materialFunction=&SmoothIntersections}.Smooth Max
    Min=@n38.Smooth Min
    Max=MaterialFunctionCall(A=@n37.R, B=@n41, K=@n36.G) {materialFunction=&SmoothIntersections}.Smooth Max
    Additive=Max(A=@n43, B=@n44)
   ) {materialFunction=&BrushBlendModes_Manual}.Result
   Alpha=Saturate(None=ScalarParameter() {defaultValue=1, parameterName=UseAngle})
  ) {constAlpha=0}
  n45 = Multiply() {desc="A=in height"}
  n46 = Subtract(A=@n45)
  n47 = VectorParameter() {parameterName=CurveParams, defaultValue=<rgba 1 1 1 1>}
  n48 = Saturate(None=Divide(A=Abs(None=@n46), B=@n47.R))
  LinearInterpolate(
   B=@n45
   Alpha=Saturate(None=MaterialFunctionCall(In=@n48, Power=@n47.B) {materialFunction=&SCurve}.Result)
  ) {desc="A=Base Terrain, B=NewHeight"}
  MaterialFunctionCall() {materialFunction=/Landmass/Landscape/OldPrototype_BP/Mats/MF/VelocityAdvect}
  MaterialFunctionCall!(Tiling) {materialFunction=/Landmass/Landscape/OldPrototype_BP/Mats/MF/CatmullRomTexture}
  Max(A=@n26) {constB=0}
  Min(A=@n13) {constB=0}
  n49 = Divide()
  n50 = Constant() {r=0.002}
  n51 = Constant() {r=0.998}
  n52 = Constant()
  Multiply(A=TextureSampleParameter2D!(UVs=Clamp(None=@n49, Min=@n50, Max=@n51), Level=@n52, Apply View MipBias) {mipValueMode=TMVM_MipLevel, parameterName=CurveRT, texture=/Landmass/Landscape/Curves/DynamicCurve_RT}.R) {constB=2}
  n53 = Divide()
  Multiply(A=TextureSampleParameter2D(UVs=Clamp(None=@n53, Min=@n50, Max=@n51), Level=@n52) {mipValueMode=TMVM_MipLevel, parameterName=CurveRT, texture=&DynamicCurve_RT}.R) {constB=2}
  n54 = VectorParameter() {parameterName=ChannelMask}
  n55 = MaterialFunctionCall() {materialFunction=/Landmass/Landscape/OldPrototype_BP/Mats/MF/PointSampleUVs}
  n56 = TextureSampleParameter2D!(UVs=@n55.TexelCenterUVs, Apply View MipBias) {parameterName=HeightRT, texture=/Engine/EngineResources/Black}
  n57 = Divide(
   A=Subtract(A=@n56.R, B=TextureSampleParameter2D(UVs=@n55.TexelCenterUVs) {parameterName=HeightRTPrev, texture=&Black}.R)
   B=CollectionParameter() {collection=&MPC_Erosion, parameterName=GridSize}
  )
  n58 = Multiply(A=@n57) {constB=-1}
  n59 = Add(
   A=Multiply(A=Max(A=@n57) {constB=0, desc="Positive = Add soil to terrain"}, B=ScalarParameter() {parameterName=Deposition})
   B=Multiply(A=Max(A=@n58) {constB=0, desc="Negative = Erode"}, B=ScalarParameter() {parameterName=Erosion})
  )
  n60 = CollectionParameter() {collection=&MPC_Erosion, parameterName=AdvectionErosion}
  n61 = LinearInterpolate(A=@n56.B, Alpha=ScalarParameter() {parameterName=ScatterParticles}) {constB=1000, desc="For scatter, Velocity has been Accumulated"}
  n62 = TextureSampleParameter2D(UVs=@n55.TexelCenterUVs) {parameterName=VelocityRT, texture=&Black}
  n63 = MaterialFunctionCall(Vector 2=@n62.RGB) {materialFunction=&VectorLength}
  n64 = CollectionParameter() {collection=&MPC_Erosion, parameterName=TotalSizeX}
  n65 = Multiply(
   A=Multiply(
    A=Divide(A=@n63.V2 Length, B=@n64)
    B=CollectionParameter() {collection=&MPC_Erosion, parameterName=deltaT}
   )
   B=ScalarParameter() {parameterName=Velocity}
  )
  n66 = Divide(A=@n62.B, B=CollectionParameter() {collection=&MPC_Erosion, parameterName=GridSize})
  n67 = Multiply(A=@n66) {constB=-1}
  n68 = Add(
   A=Multiply(A=Max(A=@n66) {constB=0, desc="Positive = Add soil to terrain"}, B=ScalarParameter() {parameterName=Deposition})
   B=Multiply(A=Max(A=@n67) {constB=0, desc="Negative = Erode"}, B=ScalarParameter() {parameterName=Erosion})
  )
  n69 = ScalarParameter() {parameterName=ScatterParticles}
  n70 = Add(
   A=LinearInterpolate(B=Multiply(A=@n59, B=@n60), Alpha=ScalarParameter() {parameterName=ScatterParticles}) {constB=0}
   B=Add(A=Multiply(A=@n61, B=@n65), B=Multiply(A=@n68, B=@n69)) {constB=0}
  )
  Multiply(A=@n70, B=@n54.RGB)
  Multiply()
  Multiply(A=@n60) {constB=100}
  OneMinus(None=Multiply(A=@n70, B=@n54.A))
  Saturate(None=@n17) {desc=Heightmap}
  n71 = Subtract(A=@n6, B=ScalarParameter() {defaultValue=0.5, parameterName=Min})
  n72 = ScalarParameter() {defaultValue=1, parameterName=WeightmapInfluence}
  Saturate(None=Add(A=Multiply(A=@n71, B=@n72), B=@n17)) {desc=Weightmap only}
 ]
}
