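# Prototype Landmass landscape brush material that stamps a textured, box-shaped foundation into the height render target.
# Position and axis parameters project BrushRT into brush space, adding scaled texture height to Z, masked by a DFBox falloff.
# A BlendMode scalar chains lerps through alpha, min, max and additive blends with HeightRT.
# Layer alpha combines the box mask, optional Z mask and texture displacement, then feeds LSBrush_ChannelSelect with the height.
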Material /Landmass/Landscape/OldPrototype_BP/Mats/Brushes/M_Foundation
graph {
 properties {
  bAllowNegativeEmissiveColor=true
  blendMode=BLEND_AlphaComposite
  pixelDepthOffsetMode=PDOM_Legacy
  shadingModel=MSM_Unlit
  usedShadingModels=Unlit
 }
 nodes [
  n1 = TextureSampleParameter2D(UVs=MaterialFunctionCall() {materialFunction=/Landmass/Landscape/OldPrototype_BP/Mats/MF/PointSampleUVs}.TexelCenterUVs) {parameterName=HeightRT, texture=/Engine/EngineResources/Black_Low.Black}
  n2 = ComponentMask(None=VectorParameter() {parameterName=X, defaultValue=<rgba 1 0 0 0>}.RGB) {g=true, r=true}
  n3 = Subtract(
   A=MaterialFunctionCall() {materialFunction=&PointSampleUVs}.TexelCenterUVs
   B=ComponentMask(None=VectorParameter() {parameterName=P}.RGB) {g=true, r=true}
  )
  n4 = VectorParameter() {parameterName=Y, defaultValue=<rgba 0 1 0 0>}
  n5 = Divide(A=ComponentMask(None=@n4.RGB) {g=true, r=true}, B=@n4.A)
  n6 = ScalarParameter() {defaultValue=256, parameterName=R}
  n7 = CollectionParameter() {collection=/Landmass/Landscape/BlueprintBrushes/MPC/MPC_Erosion, parameterName=TotalSizeX}
  n8 = CollectionParameter() {collection=&MPC_Erosion, parameterName=TotalSizeY}
  n9 = Divide(A=AppendVector(A=DotProduct(A=@n2, B=@n3), B=DotProduct(A=@n5, B=@n3)), B=Divide(A=Multiply(A=@n6) {constB=2}, B=AppendVector(A=@n7, B=@n8)))
  n10 = ScalarParameter() {defaultValue=1, parameterName=T}
  n11 = Add(A=Multiply(A=@n9, B=Multiply(A=@n10) {constB=2})) {constB=0.5}
  n12 = TextureSampleParameter2D(UVs=Saturate(None=@n11)) {parameterName=BrushRT, texture=/Engine/MapTemplates/TilingNoise05}
  n13 = OneMinus(None=@n12.R)
  n14 = Saturate(None=VectorParameter() {parameterName=X, defaultValue=<rgba 1 0 0 0>}.A)
  n15 = Add(
   A=Multiply(A=LinearInterpolate(A=@n13, B=@n12.R, Alpha=@n14), B=ScalarParameter() {defaultValue=256, parameterName=BrushTexHeight})
   B=ScalarParameter() {parameterName=Z}
  )
  n16 = ScalarParameter() {defaultValue=256, parameterName=EdgeCenterOffset}
  n17 = ScalarParameter() {defaultValue=256, parameterName=R}
  n18 = ScalarParameter() {defaultValue=256, parameterName=EdgeWidth}
  n19 = MaterialFunctionCall!(
   A=Add(A=@n9) {constB=0.5}
   B=Constant() {r=0.5}
   Bounds=Add(A=Constant() {r=0.5}, B=Divide(A=@n16, B=@n17))
   Edge Falloff=Divide(A=Divide(A=@n18, B=@n17))
  ) {materialFunction=/Landmass/Landscape/OldPrototype_BP/Mats/MF/DFBox}
  n20 = Saturate(None=@n19.Result)
  n21 = LinearInterpolate(A=@n1.R, B=@n15, Alpha=@n20)
  n22 = ScalarParameter() {parameterName=BlendMode}
  n23 = Multiply(A=@n15, B=@n20)
  n24 = Subtract(A=@n22)
  n25 = LinearInterpolate(A=LinearInterpolate(A=@n21, B=Min(A=@n1.R, B=@n21), Alpha=@n22), B=Max(A=@n1.R, B=@n23), Alpha=Saturate(None=@n24))
  n26 = Add(A=@n23, B=@n1.R)
  n27 = Saturate(None=Subtract(A=@n24))
  n28 = Subtract(
   A=TextureSampleParameter2D() {parameterName=HeightRT, texture=&Black}.R
   B=ScalarParameter() {parameterName=Z}
  )
  n29 = VectorParameter() {desc="UseZMask, MaskLength, ZOffset, UseAbs", parameterName=ZMaskParams}
  n30 = Add(A=Multiply(A=Abs(None=@n28)) {constB=-1}, B=@n29.G)
  n31 = LinearInterpolate(A=Divide(A=Subtract(A=@n28, B=@n29.B), B=@n29.G), B=Add(A=Divide(A=@n30, B=@n29.G), B=@n29.B), Alpha=@n29.A)
  n32 = TextureSampleParameter2D(UVs=@n11) {parameterName=BrushRT, texture=&TilingNoise05}
  n33 = LinearInterpolate(A=OneMinus(None=@n32.R), B=@n32.R, Alpha=Saturate(None=VectorParameter() {parameterName=X, defaultValue=<rgba 1 0 0 0>}.A))
  n34 = ScalarParameter() {defaultValue=0.5, parameterName=Displacement Midpoint}
  n35 = Add(
   A=LinearInterpolate(A=@n19.Result, B=Min(A=@n19.Result, B=@n31), Alpha=@n29.R)
   B=Multiply(A=Subtract(A=@n33, B=@n34), B=ScalarParameter() {defaultValue=1, parameterName=WeightmapInfluence})
  )
  n36 = ScalarParameter() {defaultValue=1, parameterName=BrushOpacity}
  MaterialFunctionCall!(
   Height RGB=MaterialFunctionCall(X=LinearInterpolate(A=@n25, B=@n26, Alpha=@n27), Y=@n1.G, Z=@n1.G) {materialFunction=&MakeFloat3}.Result
   Layer Alpha=Saturate(None=Multiply(A=@n35, B=@n36))
  ) {materialFunction=/Landmass/Landscape/OldPrototype_BP/Mats/MF/LSBrush_ChannelSelect}
  ScalarParameter() {parameterName=Invert}
  ScalarParameter() {defaultValue=0.2, parameterName=WeightmapZRange}
  n37 = AppendVector(A=ScalarParameter() {defaultValue=2048, parameterName=SizeX}, B=ScalarParameter() {defaultValue=2048, parameterName=SizeY})
  n38 = ScalarParameter() {defaultValue=16, parameterName=GridSize}
  SphereMask(
   A=TextureCoordinate()
   B=ComponentMask(None=VectorParameter() {parameterName=P}.RGB) {g=true, r=true}
   Radius=Divide(A=ScalarParameter() {defaultValue=256, parameterName=R}, B=Multiply(A=@n37, B=@n38))
   Hardness=ScalarParameter() {defaultValue=0.5, parameterName=H}
  )
  TextureCoordinate()
 ]
}
