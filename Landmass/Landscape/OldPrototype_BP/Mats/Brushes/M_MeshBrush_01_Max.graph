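# Prototype Landmass mesh brush material, unlit, that stamps a raised shape into a landscape heightmap using a Max blend.
# BrushRT texture, tiled, scaled and offset by Z, supplies the brush height.
# JumpFloodRT seed coordinates and alpha give a signed distance, turned into an edge mask using EdgeWidth over MPC_Erosion terrain size.
# Masked height is maxed with HeightRT red and packed with its green channel; an alternative lerp blend is left unused.
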Material /Landmass/Landscape/OldPrototype_BP/Mats/Brushes/M_MeshBrush_01_Max
graph {
 properties {
  bAllowNegativeEmissiveColor=true
  blendMode=BLEND_AlphaComposite
  pixelDepthOffsetMode=PDOM_Legacy
  shadingModel=MSM_Unlit
  usedShadingModels=Unlit
 }
 nodes [
  n1 = TextureSampleParameter2D() {parameterName=HeightRT, texture=/Engine/EngineResources/Black_Low.Black}
  n2 = Multiply(A=TextureCoordinate(), B=ScalarParameter() {defaultValue=2, parameterName=T})
  n3 = Add(
   A=Multiply(
    A=TextureSampleParameter2D(UVs=@n2) {parameterName=BrushRT, texture=/Engine/MapTemplates/TilingNoise05}.R
    B=ScalarParameter() {defaultValue=256, parameterName=BrushTexHeight}
   )
   B=ScalarParameter() {defaultValue=1394, parameterName=Z}
  )
  n4 = TextureCoordinate()
  n5 = TextureSampleParameter2D() {parameterName=JumpFloodRT, texture=&Black}
  n6 = ComponentMask(None=@n5.RGB) {g=true, r=true}
  n7 = Multiply(
   A=MaterialFunctionCall(Vector 2=Subtract(A=@n4, B=@n6)) {materialFunction=/Engine/Functions/Engine_MaterialFunctions02/Utility/VectorLength}.V2 Length
   B=LinearInterpolate(Alpha=@n5.A) {constA=1, constB=-1}
  )
  n8 = ScalarParameter() {defaultValue=1, desc=Tangent, parameterName=EdgeWidth}
  n9 = Multiply(
   A=CollectionParameter() {collection=/Landmass/Landscape/BlueprintBrushes/MPC/MPC_Erosion, parameterName=ResX}
   B=CollectionParameter() {collection=&MPC_Erosion, parameterName=GridSize}
  ) {desc=TotalSize}
  n10 = Saturate(None=Divide(A=Multiply(A=@n7) {constB=-1}, B=Divide(A=@n8, B=@n9)))
  LinearInterpolate(A=@n1.R, B=@n3, Alpha=@n10)
  MaterialFunctionCall(X=Max(A=@n1.R, B=Multiply(A=@n3, B=@n10)), Y=@n1.G, Z=@n1.G) {materialFunction=&MakeFloat3}
  ScalarParameter() {defaultValue=1394, parameterName=Z}
 ]
}
