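# Unlit material from an old Landmass landscape Blueprint prototype that defines a simple circular brush mask.
# A SphereMask compares the texture coordinates with the red and green channels of vector parameter P, which sets the brush center in UV space and defaults to the middle.
# Scalar parameters R and H control the radius and edge hardness of the circle.
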Material /Landmass/Landscape/OldPrototype_BP/Mats/Brushes/M_SimpleBrush
graph {
 properties {pixelDepthOffsetMode=PDOM_Legacy, shadingModel=MSM_Unlit, usedShadingModels=Unlit}
 nodes [SphereMask(
   A=TextureCoordinate()
   B=ComponentMask(None=VectorParameter() {parameterName=P, defaultValue=<rgba 0.5 0.5 0 0>}.RGB) {g=true, r=true}
   Radius=ScalarParameter() {defaultValue=0.125, parameterName=R}
   Hardness=ScalarParameter() {parameterName=H}
  )]
}
