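# Unlit prototype Landmass spline brush material that shapes landscape height and a channel-selected edge alpha from a cached distance field.
# Scaled distance forms falloff slopes around a Z level displaced by a brush texture.
# Smooth min/max combines these with existing height for selectable blend modes, while distance and texture drive layer alpha.
# A brute-force ring-search distance node and jump-flood neighbor averaging feed nothing downstream.
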Material /Landmass/Landscape/OldPrototype_BP/Mats/Brushes/M_SplineBrush
graph {
 properties {
  bAllowNegativeEmissiveColor=true
  blendMode=BLEND_AlphaComposite
  pixelDepthOffsetMode=PDOM_Legacy
  shadingModel=MSM_Unlit
  usedShadingModels=Unlit
 }
 nodes [
  n1 = VectorParameter() {parameterName=ChannelMask, defaultValue=<rgba 0 0 0 1>}
  AppendVector(A=@n1.RGB, B=@n1.RGB)
  MaterialExpressionTextureObjectParameter_0 = TextureObjectParameter() {parameterName=SceneCaptureRT, texture=/Engine/EngineResources/Black_Low.Black}
  MaterialExpressionAppendVector_2 = AppendVector(A=ScalarParameter() {defaultValue=2048, parameterName=SizeX}, B=ScalarParameter() {defaultValue=2048, parameterName=SizeY})
  MaterialExpressionMin_2 = Min(A=ScalarParameter() {defaultValue=32, parameterName=Radius}) {constB=64}
  MaterialExpressionTextureCoordinate_4 = TextureCoordinate()
  Custom!(
   Tex=@MaterialExpressionTextureObjectParameter_0
   TextureSize=@MaterialExpressionAppendVector_2
   MaxRadius=@MaterialExpressionMin_2
   UV=@MaterialExpressionTextureCoordinate_4
   SceneCaptureZ
  ) {
   code="\r\n\r\nfloat2 curoffset = 0;\r\nfloat curdist = 0;\r\nfloat mindist = MaxRadius;\r\nfloat texelsize = 1 / TextureSize;\r\nint laststep = 0;\r\nint newmax = MaxRadius;\r\nfloat mindistZ = 0;\r\n\r\nUV = ( floor(UV * TextureSize) / TextureSize ) + ( 0.5 / TextureSize);\r\n\r\n\r\nfloat startsample = SceneCaptureZ - Tex.SampleLevel(TexSampler, UV, 0).r;\r\nint startinside = saturate(ceil(startsample));\r\n\r\n//i is the current radius level\r\nint i = 0;\r\nwhile( i < min(MaxRadius, newmax) )\r\n{\r\ni++;\r\n\t//8 neighbors per texel radius\r\n\tfloat totneighbors = i * 8;\r\n\t//j is the current neighbor index\r\n\tint j = 0;\r\n\twhile ( j < totneighbors )\r\n\t{\r\n\t\t\r\n\t\tfloat progress = j / totneighbors;\r\n\t\tcuroffset.x=saturate(4*progress)-saturate(4*(progress-0.5));\r\n\t\tcuroffset.y=saturate(4*(progress-0.25))-saturate(4*(progress-0.75));\r\n\r\n\t\t//normalize and offset by radius\r\n\t\tcuroffset = (curoffset * totneighbors / 4) - i;\r\n\r\n\t\tfloat2 curUV = UV + curoffset * texelsize;\r\n\t\tcurUV = ( floor(curUV * TextureSize) / TextureSize ) + ( 0.5 / TextureSize);\r\n\r\n\r\n\t\tfloat cursample = SceneCaptureZ - Tex.SampleLevel(TexSampler, curUV, 0).r;\r\n\r\n\t\t//int circlemask = ( curoffset.x*curoffset.x + curoffset.y*curoffset.y );\r\n\r\n\t\tif( saturate(ceil(cursample)) != startinside )\r\n\t\t{\r\n\t\t\tcurdist = length(curoffset);\r\n\t\t\tmindist = min( mindist, curdist );\r\n\t\t\tmindistZ = cursample;\r\n\t\t}\r\n\t\t\r\n\tj++;\r\n\r\n\t}\r\n\r\n\tif ( mindist < MaxRadius &&&& newmax == MaxRadius)\r\n\t{\r\n\t\tnewmax = ceil( i * sqrt(2.0) );\t\r\n\t}\r\n\r\n\r\n}\r\n\r\nfloat2 output = float2( (mindist - 0.5) / (MaxRadius - 0.5), mindistZ);\r\n\r\nif (startinside ==1)\r\n{\r\n\toutput.x *= -1;\r\n}\r\n\r\noutput.x =  (output.x + 1) * 0.5;\r\noutput.x = 1 -output.x;\r\nreturn output;\r\n\r\n"
   outputType=CMOT_Float2
   inputs [
    {inputName=Tex, input {expression=&self:MaterialExpressionTextureObjectParameter_0}}
    {inputName=TextureSize, input {expression=&self:MaterialExpressionAppendVector_2}}
    {inputName=MaxRadius, input {expression=&self:MaterialExpressionMin_2}}
    {inputName=UV, input {expression=&self:MaterialExpressionTextureCoordinate_4}}
    {inputName=SceneCaptureZ, input {expression=null}}
   ]
  }
  n2 = TextureCoordinate()
  n3 = Divide(B=AppendVector(
    A=CollectionParameter() {collection=/Landmass/Landscape/BlueprintBrushes/MPC/MPC_Erosion, parameterName=ResX}
    B=CollectionParameter() {collection=&MPC_Erosion, parameterName=ResY}
   ))
  n4 = Add(A=@n2, B=Multiply(A=@n3, B=Constant2Vector() {r=1}))
  n5 = TextureSampleParameter2D(UVs=@n4) {parameterName=JumpFloodRT, texture=/Engine/EngineResources/Black}
  n6 = MaterialFunctionCall(Vector 2=Subtract(A=@n4, B=ComponentMask(None=@n5.RGB) {g=true, r=true})) {materialFunction=/Engine/Functions/Engine_MaterialFunctions02/Utility/VectorLength}
  n7 = LinearInterpolate(Alpha=@n5.A) {constA=1, constB=-1}
  n8 = Add(A=@n2, B=Multiply(A=@n3, B=Constant2Vector() {r=-1}))
  n9 = TextureSampleParameter2D(UVs=@n8) {parameterName=JumpFloodRT, texture=&Black}
  n10 = MaterialFunctionCall(Vector 2=Subtract(A=@n8, B=ComponentMask(None=@n9.RGB) {g=true, r=true})) {materialFunction=&VectorLength}
  n11 = LinearInterpolate(Alpha=@n9.A) {constA=1, constB=-1}
  n12 = TextureCoordinate()
  n13 = Add(A=@n12, B=Multiply(A=@n3, B=Constant2Vector() {g=1}))
  n14 = TextureSampleParameter2D(UVs=@n13) {parameterName=JumpFloodRT, texture=&Black}
  n15 = MaterialFunctionCall(Vector 2=Subtract(A=@n13, B=ComponentMask(None=@n14.RGB) {g=true, r=true})) {materialFunction=&VectorLength}
  n16 = LinearInterpolate(Alpha=@n14.A) {constA=1, constB=-1}
  n17 = Add(A=@n12, B=Multiply(A=@n3, B=Constant2Vector() {g=-1}))
  n18 = TextureSampleParameter2D(UVs=@n17) {parameterName=JumpFloodRT, texture=&Black}
  n19 = MaterialFunctionCall(Vector 2=Subtract(A=@n17, B=ComponentMask(None=@n18.RGB) {g=true, r=true})) {materialFunction=&VectorLength}
  n20 = LinearInterpolate(Alpha=@n18.A) {constA=1, constB=-1}
  n21 = Add(
   A=Add(A=Multiply(A=@n6.V2 Length, B=@n7), B=Multiply(A=@n10.V2 Length, B=@n11))
   B=Add(A=Multiply(A=@n15.V2 Length, B=@n16), B=Multiply(A=@n19.V2 Length, B=@n20))
  )
  Divide(A=@n21) {constB=4}
  n22 = TextureSampleParameter2D() {parameterName=HeightRT, texture=/Engine/EngineResources/Black_Low.Black}
  n23 = MaterialFunctionCall!(AlphaBlend=@n22.R, Min=@n22.R, Max=@n22.R, Additive=Constant()) {materialFunction=/Landmass/Landscape/OldPrototype_BP/Mats/MF/BrushBlendModes_Manual}
  n24 = MaterialFunctionCall() {materialFunction=/Landmass/Landscape/OldPrototype_BP/Mats/MF/ReadCachedDFHeight}
  n25 = Multiply(A=ScalarParameter() {defaultValue=1, desc=Tangent, parameterName=FalloffTangent}, B=@n24.UV Distance)
  n26 = LinearInterpolate(
   A=@n25
   B=MaterialFunctionCall!(A=@n25, B=Constant(), K=VectorParameter() {parameterName=SmoothingParams, defaultValue=<rgba 1 1 1 1>}.R) {materialFunction=/Landmass/Landscape/OldPrototype_BP/Mats/MF/SmoothIntersections}.Smooth Min
   Alpha=ScalarParameter() {parameterName=UseMeshDepth}
  )
  n27 = TextureCoordinate()
  n28 = ScalarParameter() {defaultValue=2, parameterName=T}
  n29 = ComponentMask(None=TextureSampleParameter2D(UVs=Multiply(A=@n27, B=@n28)) {desc="Current Texel (upper left)", parameterName=BrushRT, texture=/Engine/MapTemplates/TilingNoise05}.R) {r=true}
  n30 = ScalarParameter() {defaultValue=0.5, parameterName=Displacement Midpoint}
  n31 = Add(
   A=ScalarParameter() {defaultValue=1394, parameterName=Z}
   B=Multiply(A=Subtract(A=@n29, B=@n30), B=ScalarParameter() {defaultValue=256, parameterName=BrushTexHeight})
  )
  n32 = Add(A=Multiply(A=@n26) {constB=-1}, B=@n31)
  n33 = VectorParameter() {parameterName=SmoothingParams, defaultValue=<rgba 1 1 1 1>}
  n34 = Add(A=@n26, B=@n31)
  n35 = MaterialFunctionCall(
   A=MaterialFunctionCall(A=@n23.Result, B=@n32, K=@n33.G) {materialFunction=&SmoothIntersections}.Smooth Min
   B=@n34
   K=@n33.G
  ) {materialFunction=&SmoothIntersections}
  n36 = MaterialFunctionCall(A=@n23.Result, B=@n32, K=@n33.R) {materialFunction=&SmoothIntersections}
  n37 = MaterialFunctionCall(A=@n23.Result, B=@n34, K=@n33.G) {materialFunction=&SmoothIntersections}
  n38 = TextureSampleParameter2D() {parameterName=HeightRT, texture=&Black}
  n39 = Add(A=@n35.Smooth Max, B=@n38.R)
  n40 = ScalarParameter() {parameterName=EdgeCenterOffset}
  n41 = Add(A=Divide(A=Add(A=@n24.UV Distance, B=@n40), B=ScalarParameter() {defaultValue=1, desc=Tangent, parameterName=EdgeWidth})) {constB=0.5}
  n42 = Multiply(A=Subtract(A=@n29, B=@n30), B=ScalarParameter() {defaultValue=1, parameterName=WeightmapInfluence})
  MaterialFunctionCall!(
   Height RGB=MaterialFunctionCall(
    X=MaterialFunctionCall(AlphaBlend=@n35.Smooth Max, Min=@n36.Smooth Min, Max=@n37.Smooth Max, Additive=@n39) {materialFunction=&BrushBlendModes_Manual}.Result
    Y=@n38.G
    Z=@n38.G
   ) {materialFunction=&MakeFloat3}.Result
   Layer Alpha=Saturate(None=Add(A=@n41, B=@n42))
  ) {materialFunction=/Landmass/Landscape/OldPrototype_BP/Mats/MF/LSBrush_ChannelSelect}
  Multiply(A=Divide(A=Add(A=@n21)) {constB=5})
  ScalarParameter() {defaultValue=2000, parameterName=fsdafad}
  TextureSampleParameter2D() {parameterName=HeightRT_5, texture=/Engine/EngineResources/Black}
 ]
}
