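# Unlit prototype Landmass brush material that stamps a water volume onto a height render target.
# Center, axis and radius parameters, scaled by MPC_Erosion's total size, project BrushRT into an edge-falloff box mask.
# The scaled, Z-offset brush height minus HeightRT red feeds a manual blend function against existing blue, using its additive input.
# MakeFloat3 packs red twice plus the non-negative result; SphereMask and BlendMode branches appear unused.
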Material /Landmass/Landscape/OldPrototype_BP/Mats/Brushes/M_WaterVolume_01
graph {
 properties {
  bAllowNegativeEmissiveColor=true
  blendMode=BLEND_AlphaComposite
  pixelDepthOffsetMode=PDOM_Legacy
  shadingModel=MSM_Unlit
  usedShadingModels=Unlit
 }
 nodes [
  n1 = TextureSampleParameter2D() {parameterName=HeightRT, texture=/Engine/EngineResources/Black_Low.Black}
  n2 = Subtract(A=TextureCoordinate(), B=ComponentMask(None=VectorParameter() {parameterName=P}.RGB) {g=true, r=true})
  n3 = DotProduct(A=ComponentMask(None=VectorParameter() {parameterName=X, defaultValue=<rgba 1 0 0 0>}.RGB) {g=true, r=true}, B=@n2)
  n4 = DotProduct(A=ComponentMask(None=VectorParameter() {parameterName=Y, defaultValue=<rgba 0 1 0 0>}.RGB) {g=true, r=true}, B=@n2)
  n5 = Multiply(A=ScalarParameter() {defaultValue=256, parameterName=R}) {constB=2}
  n6 = AppendVector(
   A=CollectionParameter() {collection=/Landmass/Landscape/BlueprintBrushes/MPC/MPC_Erosion, parameterName=TotalSizeX}
   B=CollectionParameter() {collection=&MPC_Erosion, parameterName=TotalSizeY}
  )
  n7 = Add(A=Divide(A=AppendVector(A=@n3, B=@n4), B=Divide(A=@n5, B=@n6))) {constB=0.5}
  n8 = Saturate(None=@n7)
  n9 = Add(
   A=Multiply(
    A=TextureSampleParameter2D(UVs=@n8) {parameterName=BrushRT, texture=/Engine/MapTemplates/TilingNoise05}.R
    B=ScalarParameter() {defaultValue=256, parameterName=BrushTexHeight}
   )
   B=ScalarParameter() {parameterName=Z}
  )
  n10 = ScalarParameter() {defaultValue=256, parameterName=EdgeWidth}
  n11 = ScalarParameter() {defaultValue=256, parameterName=R}
  n12 = MaterialFunctionCall!(A=@n7, B=Constant() {r=0.5}, Bounds=Constant() {r=0.5}, Edge Falloff=Divide(A=Divide(A=@n10, B=@n11))) {materialFunction=/Engine/Functions/Engine_MaterialFunctions01/Gradient/BoxMask-2D}
  n13 = LinearInterpolate(A=@n1.B, B=@n9, Alpha=@n12.Result)
  n14 = Subtract(A=Multiply(A=@n9, B=@n12.Result), B=@n1.R)
  n15 = Min(A=@n1.B, B=@n14)
  n16 = Max(A=@n14, B=@n1.B) {constB=0}
  MaterialFunctionCall(
   X=@n1.R
   Y=@n1.R
   Z=Max(A=MaterialFunctionCall!(AlphaBlend=@n13, Min=@n15, Max=@n16, Additive) {materialFunction=/Landmass/Landscape/OldPrototype_BP/Mats/MF/BrushBlendModes_Manual}.Result) {constB=0}
  ) {materialFunction=&MakeFloat3}
  Saturate(None=Subtract(A=ScalarParameter() {parameterName=BlendMode}))
  n17 = AppendVector(A=ScalarParameter() {defaultValue=2048, parameterName=SizeX}, B=ScalarParameter() {defaultValue=2048, parameterName=SizeY})
  n18 = ScalarParameter() {defaultValue=16, parameterName=GridSize}
  SphereMask(
   A=TextureCoordinate()
   B=ComponentMask(None=VectorParameter() {parameterName=P}.RGB) {g=true, r=true}
   Radius=Divide(A=ScalarParameter() {defaultValue=256, parameterName=R}, B=Multiply(A=@n17, B=@n18))
   Hardness=ScalarParameter() {defaultValue=0.5, parameterName=H}
  )
 ]
}
