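# Unlit Landmass prototype brush material that adds tileable fractal Voronoi noise to a landscape height render target.
# A Custom node loops Noise3D_Multiplexer over up to ten octaves, with UVs scaled by MPC_Erosion component count and Tiling, and Seed as Z.
# The optionally inverted, pre-offset noise maps to Elevation Min/Max, adds to point-sampled HeightRT red, and clamps to signed 16-bit range.
# Separate RG8 pack/unpack debug nodes and a stock Noise node sit outside that chain.
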Material /Landmass/Landscape/OldPrototype_BP/Mats/Brushes/NoiseBrushVoronoi
graph {
 properties {bAllowNegativeEmissiveColor=true, pixelDepthOffsetMode=PDOM_Legacy, shadingModel=MSM_Unlit, usedShadingModels=Unlit}
 nodes [
  n1 = TextureCoordinate()
  n2 = AppendVector(
   A=CollectionParameter() {collection=/Landmass/Landscape/BlueprintBrushes/MPC/MPC_Erosion, parameterName=ResX}
   B=CollectionParameter() {collection=&MPC_Erosion, parameterName=ResY}
  )
  n3 = CollectionParameter() {collection=&MPC_Erosion, parameterName=ComponentRes}
  n4 = Floor(None=Divide(A=Subtract(A=@n2), B=Subtract(A=@n3)) {constB=256})
  n5 = ScalarParameter() {defaultValue=2, parameterName=Tiling}
  MaterialExpressionAppendVector_0 = AppendVector(A=Multiply(A=Multiply(A=@n1, B=@n4), B=@n5), B=ScalarParameter() {parameterName=Seed})
  MaterialExpressionConstant_2 = Constant() {desc="3 = Gradient, 5 = Voronoi", r=5}
  MaterialExpressionConstant_3 = Constant()
  MaterialExpressionMin_0 = Min(A=ScalarParameter() {defaultValue=3, parameterName=Octaves}) {constB=10}
  MaterialExpressionScalarParameter_2 = ScalarParameter() {defaultValue=0.5, parameterName=Persistence}
  MaterialExpressionMultiply_9 = Multiply(A=@n4, B=@n5)
  n6 = Custom!(
   Position=@MaterialExpressionAppendVector_0
   Function=@MaterialExpressionConstant_2
   bTurbulence=@MaterialExpressionConstant_3
   Octaves=@MaterialExpressionMin_0
   Persistence=@MaterialExpressionScalarParameter_2
   RepeatSize=@MaterialExpressionMultiply_9
  ) {
   code="\r\nfloat Out = 0;\r\nfloat OutScale = 1;\r\n\r\n[loop]\r\nfor(int i = 0; i < Octaves; ++i)\r\n{\r\n\r\n\r\n\tif(bTurbulence)\r\n\t{\r\n\t\tOut += abs(Noise3D_Multiplexer(Function, Position, 4, 1, RepeatSize)) * OutScale;\r\n\t}\r\n\telse\r\n\t{\r\n\t\tOut += Noise3D_Multiplexer(Function, Position, 4, 1, RepeatSize) * OutScale;\r\n\t}\r\n\r\n\tPosition *= 2;\r\n\tRepeatSize *= 2;\r\n\tOutScale *= Persistence;\r\n\r\n}\r\n\r\nif(!bTurbulence)\r\n{\r\n\t// bring -1..1 to 0..1 range\r\n\tOut = Out * 0.5f + 0.5f;\r\n}\r\n\r\n// Out is in 0..1 range\r\nreturn lerp(0, 1, Out);"
   description=CustomNoise
   outputType=CMOT_Float1
   inputs [
    {inputName=Position, input {expression=&self:MaterialExpressionAppendVector_0}}
    {inputName=Function, input {expression=&self:MaterialExpressionConstant_2}}
    {inputName=bTurbulence, input {expression=&self:MaterialExpressionConstant_3}}
    {inputName=Octaves, input {expression=&self:MaterialExpressionMin_0}}
    {
     inputName=Persistence
     input {expression=&self:MaterialExpressionScalarParameter_2}
    }
    {inputName=RepeatSize, input {expression=&self:MaterialExpressionMultiply_9}}
   ]
  }
  n7 = LinearInterpolate(A=@n6, B=OneMinus(None=@n6), Alpha=ScalarParameter() {defaultValue=1, parameterName=Invert})
  n8 = ScalarParameter() {parameterName=Elevation Pre-Offset}
  n9 = LinearInterpolate(
   A=ScalarParameter() {parameterName=Elevation Min}
   B=ScalarParameter() {defaultValue=10000, parameterName=Elevation Max}
   Alpha=Max(A=Subtract(A=@n7, B=@n8)) {constB=0}
  )
  n10 = TextureSampleParameter2D(UVs=MaterialFunctionCall() {materialFunction=/Landmass/Landscape/OldPrototype_BP/Mats/MF/PointSampleUVs}.TexelCenterUVs) {parameterName=HeightRT, texture=/Engine/EngineResources/Black_Low.Black}
  Clamp(None=MaterialFunctionCall(X=Add(A=@n9, B=@n10.R), Y=@n10.G, Z=@n10.G) {materialFunction=&MakeFloat3}.Result) {maxDefault=32767, minDefault=-32768}
  n11 = ScalarParameter() {defaultValue=-32768, parameterName=H}
  n12 = MaterialFunctionCall!(Height Int -32k to 32k=@n11) {materialFunction=/Landmass/Landscape/OldPrototype_BP/Mats/MF/PackRG8Height}
  Divide(A=@n12.XY) {constB=255}
  n13 = Floor(None=Add(A=@n11) {constB=32768, desc="Make  0-65536"})
  n14 = Floor(None=Divide(A=@n13) {constB=256})
  n15 = Fmod(A=@n13, B=Constant() {r=256})
  n16 = Divide(A=Floor(None=AppendVector(A=@n14, B=@n15))) {constB=255}
  MaterialFunctionCall!(Vector2=@n16, MaximumNumberOfDigits=Constant() {r=8}, UVs, DebugTextLocation RG_UpperRight BA_LowerLeft, Component Spacing) {materialFunction=/Engine/Functions/Engine_MaterialFunctions02/Utility/DebugFloat2Values}
  Multiply(A=@n12.XY) {constB=255}
  Noise!(World Position=@MaterialExpressionAppendVector_0, FilterWidth) {bTurbulence=false, levelScale=3, levels=16, noiseFunction=NOISEFUNCTION_VoronoiALU, outputMin=0, quality=5, repeatSize=2}
  ScalarParameter() {defaultValue=5, parameterName=NoiseType}
  Sign(None=MaterialFunctionCall!(RG=@n16) {materialFunction=/Landmass/Landscape/OldPrototype_BP/Mats/MF/UnpackRG8Height}.Result)
 ]
}
